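; Example PCG graph that builds a 400x50 m row of basic city buildings from chained PCGPrimitives subgraphs.
; A spline line becomes an outline, is recursively split into lots, and gets 50-100 m heights from a texture.
; The lots branch into shape grammar assignment and a double offset that feeds planar meshes.
; The offset areas are also contracted, filled uniformly and randomly filtered to spawn assets.
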(graph Basic_Building_Row_of_400x50m
  (bind Subgraph_0 (Subgraph :Graph /PCGPrimitives/Primitives/Create/Create_Spline_line.Create_Spline_line))
  (bind Subgraph_1 (Subgraph :Graph /PCGPrimitives/Primitives/Transform/Transform_Outline.Transform_Outline))
  (bind Subgraph_2 (Subgraph :Graph /PCGPrimitives/Primitives/Subdivide/Subdivide_Areas_WithRecursiveSplit.Subdivide_Areas_WithRecursiveSplit))
  (bind Subgraph_3 (Subgraph :Graph /PCGPrimitives/Primitives/Write/Write_Attribute_fromTexture.Write_Attribute_fromTexture)) ; Height will vary between 50 and 100 meters
  (bind Subgraph_6 (Subgraph :Graph /PCGPrimitives/Primitives/Assign/Assign_ShapeGrammarDefinition.Assign_ShapeGrammarDefinition))
  (bind Subgraph_4 (Subgraph :Graph /PCGPrimitives/Primitives/Transform/Transform_Offset.Transform_Offset))
  (bind Subgraph_7 (Subgraph :Graph /PCGPrimitives/Primitives/Transform/Transform_Offset.Transform_Offset))
  (bind Subgraph_5 (Subgraph :Graph /PCGPrimitives/Primitives/Create/Create_Mesh_Planar.Create_Mesh_Planar))
  (bind Subgraph_12 (Subgraph :Graph /PCGPrimitives/Primitives/Transform/Transform_Areas_Contract.Transform_Areas_Contract))
  (bind Subgraph_9 (Subgraph :Graph /PCGPrimitives/Primitives/Fill/Fill_Areas_Uniformly.Fill_Areas_Uniformly))
  (bind Subgraph_10 (Subgraph :Graph /PCGPrimitives/Primitives/Filter/Filter_RandomChoice.Filter_RandomChoice))
  (bind Subgraph_11 (Subgraph :Graph /PCGPrimitives/Primitives/Spawn/Spawn_Assets.Spawn_Assets))
  (connect Subgraph_0 Out Subgraph_1 PrimaryInput)
  (connect Subgraph_1 Out Subgraph_2 PrimaryInput)
  (connect Subgraph_2 Areas Subgraph_3 PrimaryInput)
  (connect Subgraph_3 Out Subgraph_4 PrimaryInput)
  (connect Subgraph_7 Out Subgraph_5 PrimaryInput)
  (connect Subgraph_3 Out Subgraph_6 PrimaryInput)
  (connect Subgraph_4 Out Subgraph_7 PrimaryInput)
  (connect Subgraph_12 Out Subgraph_9 PrimaryInput)
  (connect Subgraph_9 Out Subgraph_10 PrimaryInput)
  (connect Subgraph_10 Selected Subgraph_11 PrimaryInput)
  (connect Subgraph_7 Out Subgraph_12 PrimaryInput))
