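; PCG example graph that generates a medieval European walled city from a jittered spline outline and two jittered arteries.
; Arteries, recursive splits and grids subdivide districts, with Disturbance and AngleJitter scaled per level.
; Road footprints by hierarchy are unioned and excluded from the city, leaving lots that become extruded buildings, some around inner courts, or randomly chosen planted plazas.
; Walls and towers are copied along the expanded outline.
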(graph City_European_Medieval_1km_2_arteries
  (params (LocalRoadWidth :type Float :default 5.5e+02) (CollectorRoadWidth :type Float :default 1e+03) (ArteryRoadWidth :type Float :default 2.2e+03) (BlockWidth :type Float :default 6e+03) (BlockLength :type Float :default 4e+03) (DistrictSizeLimit :type Float :default 4e+04) (BuildableLotsRatio :type Float :default 0.9) (Disturbance :type Float :default 0.5) (AngleJitter :type Float :default 0.5))
  (bind Subgraph_0 (Subgraph :Graph /PCGPrimitives/Primitives/Create/Create_Spline_Shape.Create_Spline_Shape))
  (bind Subgraph_2 (Subgraph :Graph /PCGPrimitives/Primitives/Create/Create_Spline_line.Create_Spline_line))
  (bind Subgraph_3 (Subgraph :Graph /PCGPrimitives/Primitives/Create/Create_Spline_line.Create_Spline_line))
  (bind Subgraph_42 (Subgraph :Graph /PCGPrimitives/Primitives/Transform/Transform_Jitter.Transform_Jitter))
  (bind WallExpand (Subgraph :Graph /PCGPrimitives/Primitives/Transform/Transform_Areas_Expand.Transform_Areas_Expand))
  ;[0.15 0.15 0.15 0.5] DISTRICTS
  (bind Subgraph_4 (Subgraph :Graph /PCGPrimitives/Primitives/Subdivide/Subdivide_Areas_withSpline.Subdivide_Areas_withSpline))
  (bind Subgraph_6 (Subgraph :Graph /PCGPrimitives/Primitives/Subdivide/Subdivide_Areas_WithRecursiveSplit.Subdivide_Areas_WithRecursiveSplit))
  (bind Subgraph_7 (Subgraph :Graph /PCGPrimitives/Primitives/Filter/Filter_Areas_BySize.Filter_Areas_BySize :Size DistrictSizeLimit))
  (bind Subgraph_5 (Subgraph :Graph /PCGPrimitives/Primitives/Subdivide/Subdivide_Areas_withGrid.Subdivide_Areas_withGrid :GridLength BlockLength :GridWidth BlockWidth))
  (bind Subgraph_39 (Subgraph :Graph /PCGPrimitives/Primitives/Subdivide/Subdivide_Areas_WithRecursiveSplit.Subdivide_Areas_WithRecursiveSplit))
  ;[0.15 0.15 0.15 0.5] DISTRUBANCE PARMETER
  (bind Const_DistrictBlocksDistMax (Create%20Constant :AttributeTypes (struct :DoubleValue 3e+03))) ; 2x current baseline (1500)
  (bind Mul_DistrictBlocksDist (Attribute%20Maths%20Op :InputSource1 Disturbance :InputSource2 None :Operation Multiply :OutputTarget Disturbance :InA Disturbance)) ; Out drives District Blocks.disturbance
  (bind Const_SubDistrictsDistMax (Create%20Constant :AttributeTypes (struct :DoubleValue 8e+03))) ; 2x current baseline (4000)
  (bind Mul_SubDistrictsDist (Attribute%20Maths%20Op :InputSource1 Disturbance :InputSource2 None :Operation Multiply :OutputTarget Disturbance :InA Disturbance)) ; Out drives Sub Districts.disturbance
  (bind Const_SmallerSubDistDistMax (Create%20Constant :AttributeTypes (struct :DoubleValue 5e+03))) ; 2x current baseline (2500)
  (bind Mul_SmallerSubDistDist (Attribute%20Maths%20Op :InputSource1 Disturbance :InputSource2 None :Operation Multiply :OutputTarget Disturbance :InA Disturbance)) ; Out drives Smaller Sub-districts.disturbance
  ;[0.15 0.15 0.15 0.5] ROADS
  (bind Subgraph_28 (Subgraph :Graph /PCGPrimitives/Primitives/Assign/Assign_ShapeGrammarDefinition.Assign_ShapeGrammarDefinition :Size CollectorRoadWidth))
  (bind Subgraph_33 (Subgraph :Graph /PCGPrimitives/Primitives/Assign/Assign_ShapeGrammarDefinition.Assign_ShapeGrammarDefinition :Size LocalRoadWidth))
  (bind Subgraph_38 (Subgraph :Graph /PCGPrimitives/Primitives/Assign/Assign_ShapeGrammarDefinition.Assign_ShapeGrammarDefinition :Size ArteryRoadWidth))
  (bind Subgraph_27 (Subgraph :Graph /PCGPrimitives/Primitives/Compose/Compose_Areas_Union.Compose_Areas_Union))
  ;[0.15 0.15 0.15 0.5] LOTS
  (bind Subgraph_12 (Subgraph :Graph /PCGPrimitives/Primitives/Compose/Compose_Areas_Exclusion.Compose_Areas_Exclusion))
  (bind Subgraph_30 (Subgraph :Graph /PCGPrimitives/Primitives/Filter/Filter_RandomChoice.Filter_RandomChoice :Ratio BuildableLotsRatio)) ; We exlcude randomly 10% of the lots, as we will later use them,\nTo build plazas.
  (bind Subgraph_13 (Subgraph :Graph /PCGPrimitives/Primitives/Filter/Filter_Areas_BySize.Filter_Areas_BySize))
  ;[0.15 0.15 0.15 0.5 50] BUILDINGS
  (bind Subgraph_14 (Subgraph :Graph /PCGPrimitives/Primitives/Subdivide/Subdivide_Areas_WithRecursiveSplit.Subdivide_Areas_WithRecursiveSplit))
  (bind Subgraph_15 (Subgraph :Graph /PCGPrimitives/Primitives/Filter/Filter_Areas_BySize.Filter_Areas_BySize))
  (bind Subgraph_51 (Subgraph :Graph /PCGPrimitives/Primitives/Filter/Filter_Areas_BySize.Filter_Areas_BySize))
  (bind Subgraph_49 (Subgraph :Graph /PCGPrimitives/Primitives/Transform/Transform_Areas_Contract.Transform_Areas_Contract))
  (bind Subgraph_65 (Subgraph :Graph /PCGPrimitives/Primitives/Subdivide/Subdivide_Areas_withGrid.Subdivide_Areas_withGrid))
  (bind Subgraph_67 (Subgraph :Graph /PCGPrimitives/Primitives/Compose/Compose_Areas_Exclusion.Compose_Areas_Exclusion))
  (bind Subgraph_66 (Subgraph :Graph /PCGPrimitives/Primitives/Subdivide/Subdivide_Areas_WithRecursiveSplit.Subdivide_Areas_WithRecursiveSplit))
  (bind Subgraph_68 (Subgraph :Graph /PCGPrimitives/Primitives/Filter/Filter_Areas_BySize.Filter_Areas_BySize))
  (bind Subgraph_21 (Subgraph :Graph /PCGPrimitives/Primitives/Create/Create_Mesh_Extrude.Create_Mesh_Extrude))
  (bind Subgraph_69 (Subgraph :Graph /PCGPrimitives/Primitives/Create/Create_Mesh_Planar.Create_Mesh_Planar))
  (bind Subgraph_11 (Subgraph :Graph /PCGPrimitives/Primitives/Filter/Filter_Areas_BySize.Filter_Areas_BySize))
  (bind Subgraph_10 (Subgraph :Graph /PCGPrimitives/Primitives/Create/Create_Mesh_Planar.Create_Mesh_Planar))
  ;[0.05488 0.151 0.03147 0.5 50] Greenery
  (bind Subgraph_23 (Subgraph :Graph /PCGPrimitives/Primitives/Fill/Fill_Areas_Randomly.Fill_Areas_Randomly))
  (bind Subgraph_24 (Subgraph :Graph /PCGPrimitives/Primitives/Spawn/Spawn_Assets.Spawn_Assets))
  (bind Subgraph_48 (Subgraph :Graph /PCGPrimitives/Primitives/Create/Create_Mesh_Planar.Create_Mesh_Planar))
  ;[0.15 0.15 0.15 0.5] ARTERIES
  (bind Subgraph_1 (Subgraph :Graph /PCGPrimitives/Primitives/Transform/Transform_Resample.Transform_Resample))
  (bind Subgraph_35 (Subgraph :Graph /PCGPrimitives/Primitives/Transform/Transform_Jitter.Transform_Jitter))
  (bind Subgraph_40 (Subgraph :Graph /PCGPrimitives/Primitives/Transform/Transform_Resample.Transform_Resample))
  (bind Subgraph_41 (Subgraph :Graph /PCGPrimitives/Primitives/Transform/Transform_Jitter.Transform_Jitter))
  ;[0.15 0.15 0.15 0.5] BORDER WALL
  (bind Subgraph_52 (Subgraph :Graph /PCGPrimitives/Primitives/Create/Create_Point.Create_Point))
  (bind Subgraph_54 (Subgraph :Graph /PCGPrimitives/Primitives/Place/Place_OnSpline.Place_OnSpline))
  (bind Subgraph_55 (Subgraph :Graph /PCGPrimitives/Primitives/Copy/CopyToPoints.CopyToPoints))
  (bind Subgraph_56 (Subgraph :Graph /PCGPrimitives/Primitives/Spawn/Spawn_Assets.Spawn_Assets))
  ;[0.15 0.15 0.15 0.5] TOWERS
  (bind Subgraph_57 (Subgraph :Graph /PCGPrimitives/Primitives/Create/Create_Point.Create_Point))
  (bind Subgraph_58 (Subgraph :Graph /PCGPrimitives/Primitives/Place/Place_OnSpline.Place_OnSpline))
  (bind Subgraph_59 (Subgraph :Graph /PCGPrimitives/Primitives/Copy/CopyToPoints.CopyToPoints))
  (bind Subgraph_60 (Subgraph :Graph /PCGPrimitives/Primitives/Spawn/Spawn_Assets.Spawn_Assets))
  (bind Subgraph_61 (Subgraph :Graph /PCGPrimitives/Primitives/Transform/Transform_Offset.Transform_Offset))
  (bind Subgraph_62 (Subgraph :Graph /PCGPrimitives/Primitives/Create/Create_Point.Create_Point))
  (bind Subgraph_63 (Subgraph :Graph /PCGPrimitives/Primitives/Copy/CopyToPoints.CopyToPoints))
  (bind Subgraph_64 (Subgraph :Graph /PCGPrimitives/Primitives/Spawn/Spawn_Assets.Spawn_Assets))
  ;[0.15 0.15 0.15 0.5] ANGLE JITTER PARARMETER
  (bind Const_DistrictBlocksJitterMax (Create%20Constant :AttributeTypes (struct :DoubleValue 1e+01))) ; 2x baseline (5)
  (bind Const_SubDistrictsJitterMax (Create%20Constant :AttributeTypes (struct :DoubleValue 5e+01))) ; 2x baseline (25)
  (bind Const_SmallerSubDistJitterMax (Create%20Constant :AttributeTypes (struct :DoubleValue 5e+01))) ; 2x baseline (25)
  (bind Mul_DistrictBlocksJitter (Attribute%20Maths%20Op :InputSource1 AngleJitter :InputSource2 None :Operation Multiply :OutputTarget AngleJitter :InA AngleJitter)) ; Out drives District Blocks.angle_Jitter
  (bind Mul_SubDistrictsJitter (Attribute%20Maths%20Op :InputSource1 AngleJitter :InputSource2 None :Operation Multiply :OutputTarget AngleJitter :InA AngleJitter)) ; Out drives Sub Districts.angle_Jitter
  (bind Mul_SmallerSubDistJitter (Attribute%20Maths%20Op :InputSource1 AngleJitter :InputSource2 None :Operation Multiply :OutputTarget AngleJitter :InA AngleJitter)) ; Out drives Smaller Sub-districts.angle_Jitter
  (reroute Adjusted%20city%20shape%20 Subgraph_42)
  (reroute Artery%201%20%20Out Subgraph_35)
  (reroute Artery%202%20Out Subgraph_41)
  (reroute NamedRerouteDeclaration_5 Subgraph_68.Unselected)
  (reroute NamedRerouteDeclaration_6 Subgraph_30.Unselected)
  (reroute NamedRerouteDeclaration_7 Subgraph_15.Unselected)
  (connect Const_DistrictBlocksDistMax Out Mul_DistrictBlocksDist InB)
  (connect Const_DistrictBlocksJitterMax Out Mul_DistrictBlocksJitter InB)
  (connect Const_SmallerSubDistDistMax Out Mul_SmallerSubDistDist InB)
  (connect Const_SmallerSubDistJitterMax Out Mul_SmallerSubDistJitter InB)
  (connect Const_SubDistrictsDistMax Out Mul_SubDistrictsDist InB)
  (connect Const_SubDistrictsJitterMax Out Mul_SubDistrictsJitter InB)
  (connect Subgraph_2 Out Subgraph_1 PrimaryInput)
  (connect Adjusted%20city%20shape%20 Out Subgraph_4 PrimaryInput)
  (connect Artery%201%20%20Out Out Subgraph_4 SecondaryInput)
  (connect Artery%202%20Out Out Subgraph_4 SecondaryInput)
  (connect Mul_DistrictBlocksDist Out Subgraph_5 Disturbance)
  (connect Subgraph_7 Selected Subgraph_5 PrimaryInput)
  (connect Mul_SubDistrictsDist Out Subgraph_6 Disturbance)
  (connect Subgraph_4 Areas Subgraph_6 PrimaryInput)
  (connect Subgraph_6 Areas Subgraph_7 PrimaryInput)
  (connect Subgraph_11 Selected Subgraph_10 PrimaryInput)
  (connect NamedRerouteDeclaration_7 Out Subgraph_11 PrimaryInput)
  (connect Subgraph_68 Unselected Subgraph_11 PrimaryInput)
  (connect Adjusted%20city%20shape%20 Out Subgraph_12 PrimaryInput)
  (connect Subgraph_27 Out Subgraph_12 SecondaryInput)
  (connect Subgraph_30 Selected Subgraph_13 PrimaryInput)
  (connect Subgraph_13 Unselected Subgraph_14 PrimaryInput)
  (connect Subgraph_14 Areas Subgraph_15 PrimaryInput)
  (connect Subgraph_51 Selected Subgraph_21 PrimaryInput)
  (connect Subgraph_68 Selected Subgraph_21 PrimaryInput)
  (connect NamedRerouteDeclaration_6 Out Subgraph_23 PrimaryInput)
  (connect Subgraph_23 Out Subgraph_24 PrimaryInput)
  (connect Subgraph_28 Footprints Subgraph_27 PrimaryInput)
  (connect Subgraph_33 Footprints Subgraph_27 PrimaryInput)
  (connect Subgraph_38 Footprints Subgraph_27 PrimaryInput)
  (connect Subgraph_6 Subdivide%20Splines Subgraph_28 PrimaryInput)
  (connect Subgraph_12 Out Subgraph_30 PrimaryInput)
  (connect Subgraph_5 Subdividing_splines Subgraph_33 PrimaryInput)
  (connect Subgraph_39 Subdivide%20Splines Subgraph_33 PrimaryInput)
  (connect Subgraph_1 Out Subgraph_35 PrimaryInput)
  (connect Adjusted%20city%20shape%20 Out Subgraph_38 PrimaryInput)
  (connect Subgraph_4 Subdividing%20Splines Subgraph_38 PrimaryInput)
  (connect Mul_SmallerSubDistDist Out Subgraph_39 Disturbance)
  (connect Subgraph_7 Unselected Subgraph_39 PrimaryInput)
  (connect Subgraph_3 Out Subgraph_40 PrimaryInput)
  (connect Subgraph_40 Out Subgraph_41 PrimaryInput)
  (connect Subgraph_0 Out Subgraph_42 PrimaryInput)
  (connect NamedRerouteDeclaration_6 Out Subgraph_48 PrimaryInput)
  (connect Subgraph_13 Selected Subgraph_49 PrimaryInput)
  (connect Subgraph_15 Selected Subgraph_51 PrimaryInput)
  (connect WallExpand Out Subgraph_54 PrimaryInput)
  (connect Subgraph_52 Out Subgraph_55 Source)
  (connect Subgraph_54 Out Subgraph_55 Target)
  (connect Subgraph_55 Out Subgraph_56 PrimaryInput)
  (connect WallExpand Out Subgraph_58 PrimaryInput)
  (connect Subgraph_57 Out Subgraph_59 Source)
  (connect Subgraph_58 Out Subgraph_59 Target)
  (connect Subgraph_59 Out Subgraph_60 PrimaryInput)
  (connect Subgraph_58 Out Subgraph_61 PrimaryInput)
  (connect Subgraph_62 Out Subgraph_63 Source)
  (connect Subgraph_61 Out Subgraph_63 Target)
  (connect Subgraph_63 Out Subgraph_64 PrimaryInput)
  (connect Subgraph_13 Selected Subgraph_65 PrimaryInput)
  (connect Subgraph_67 Out Subgraph_66 PrimaryInput)
  (connect Subgraph_65 Areas Subgraph_67 PrimaryInput)
  (connect Subgraph_49 Out Subgraph_67 SecondaryInput)
  (connect Subgraph_66 Areas Subgraph_68 PrimaryInput)
  (connect Subgraph_49 Out Subgraph_69 PrimaryInput)
  (connect Adjusted%20city%20shape%20 Out WallExpand PrimaryInput))
