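; PCG example graph for the base step of a city generator, spawning spline actors for a base shape and two arteries.
; It is built entirely from PCGPrimitives subgraphs in three independent chains.
; The shape chain creates a spline shape, jitters it and spawns it.
; Each artery chain creates a spline line, resamples and jitters it, then spawns it, with the second artery also offset before spawning.
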(graph City_Generator_Steps_1_Base
  ;[0.15 0.15 0.15 0.5] Comment
  (bind Subgraph_0 (Subgraph :Graph /PCGPrimitives/Primitives/Create/Create_Spline_Shape.Create_Spline_Shape))
  (bind Subgraph_6 (Subgraph :Graph /PCGPrimitives/Primitives/Transform/Transform_Jitter.Transform_Jitter))
  (bind Subgraph_10 (Subgraph :Graph /PCGPrimitives/Primitives/Spawn/Spawn_SplineActor.Spawn_SplineActor))
  ;[0.15 0.15 0.15 0.5] ARTERIES
  (bind Subgraph_2 (Subgraph :Graph /PCGPrimitives/Primitives/Create/Create_Spline_line.Create_Spline_line))
  (bind Subgraph_3 (Subgraph :Graph /PCGPrimitives/Primitives/Create/Create_Spline_line.Create_Spline_line))
  (bind Subgraph_1 (Subgraph :Graph /PCGPrimitives/Primitives/Transform/Transform_Resample.Transform_Resample))
  (bind Subgraph_4 (Subgraph :Graph /PCGPrimitives/Primitives/Transform/Transform_Jitter.Transform_Jitter))
  (bind Subgraph_5 (Subgraph :Graph /PCGPrimitives/Primitives/Spawn/Spawn_SplineActor.Spawn_SplineActor))
  (bind Subgraph_7 (Subgraph :Graph /PCGPrimitives/Primitives/Transform/Transform_Resample.Transform_Resample))
  (bind Subgraph_8 (Subgraph :Graph /PCGPrimitives/Primitives/Transform/Transform_Jitter.Transform_Jitter))
  (bind Subgraph_11 (Subgraph :Graph /PCGPrimitives/Primitives/Transform/Transform_Offset.Transform_Offset))
  (bind Subgraph_9 (Subgraph :Graph /PCGPrimitives/Primitives/Spawn/Spawn_SplineActor.Spawn_SplineActor))
  (connect Subgraph_2 Out Subgraph_1 PrimaryInput)
  (connect Subgraph_1 Out Subgraph_4 PrimaryInput)
  (connect Subgraph_4 Out Subgraph_5 PrimaryInput)
  (connect Subgraph_0 Out Subgraph_6 PrimaryInput)
  (connect Subgraph_3 Out Subgraph_7 PrimaryInput)
  (connect Subgraph_7 Out Subgraph_8 PrimaryInput)
  (connect Subgraph_11 Out Subgraph_9 PrimaryInput)
  (connect Subgraph_6 Out Subgraph_10 PrimaryInput)
  (connect Subgraph_8 Out Subgraph_11 PrimaryInput))
