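; PCG city generator example step that divides a city shape into districts and derives a three-tier road hierarchy.
; The tagged shape is cut by input splines, recursively split, then filtered by district size; selected areas get a block-sized grid.
; Shape outline and cutting splines become artery roads, recursive splits collectors, grid lines local roads, each with a width parameter.
; Tagged road footprints and unselected areas are output; grid cells are not.
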(graph City_Generator_Steps_2_Districts
  (params (LocalRoadWidth :type Float :default 1.9e+03) (CollectorRoadWidth :type Float :default 2.6e+03) (ArteryRoadWidth :type Float :default 3.4e+03) (BlockWidth :type Float :default 1.2e+04) (BlockLength :type Float :default 6e+03) (DistrictSizeLimit :type Float :default 3.2e+04))
  (bind Subgraph_0 (Subgraph :Graph /PCGPrimitives/Primitives/Get/Get_ActorData.Get_ActorData))
  (bind DeleteTags_0 (Delete%20Tags :Operation KeepOnlySelectedTags))
  (bind AddTags_0 (Add%20Tags :TagsToAdd "CityShape"))
  (bind Subgraph_1 (Subgraph :Graph /PCGPrimitives/Primitives/Get/Get_ActorData.Get_ActorData))
  (bind Subgraph_9 (Subgraph :Graph /PCGPrimitives/Primitives/Assign/Assign_ShapeGrammarDefinition.Assign_ShapeGrammarDefinition :Size CollectorRoadWidth))
  (bind Subgraph_10 (Subgraph :Graph /PCGPrimitives/Primitives/Assign/Assign_ShapeGrammarDefinition.Assign_ShapeGrammarDefinition :Size LocalRoadWidth))
  (bind Subgraph_13 (Subgraph :Graph /PCGPrimitives/Primitives/Assign/Assign_ShapeGrammarDefinition.Assign_ShapeGrammarDefinition :Size ArteryRoadWidth))
  (bind Subgraph_2 (Subgraph :Graph /PCGPrimitives/Primitives/Write/Write_Tag.Write_Tag))
  (bind Subgraph_3 (Subgraph :Graph /PCGPrimitives/Primitives/Write/Write_Tag.Write_Tag))
  ;[0.15 0.15 0.15 0.5] DISTRICTS
  (bind Subgraph_5 (Subgraph :Graph /PCGPrimitives/Primitives/Subdivide/Subdivide_Areas_withSpline.Subdivide_Areas_withSpline))
  (bind Subgraph_7 (Subgraph :Graph /PCGPrimitives/Primitives/Subdivide/Subdivide_Areas_WithRecursiveSplit.Subdivide_Areas_WithRecursiveSplit))
  (bind Subgraph_8 (Subgraph :Graph /PCGPrimitives/Primitives/Filter/Filter_Areas_BySize.Filter_Areas_BySize :Size DistrictSizeLimit))
  (bind Subgraph_6 (Subgraph :Graph /PCGPrimitives/Primitives/Subdivide/Subdivide_Areas_withGrid.Subdivide_Areas_withGrid :GridLength BlockLength :GridWidth BlockWidth))
  (connect DeleteTags_0 Out AddTags_0 In)
  (connect Subgraph_0 Out DeleteTags_0 In)
  (connect Subgraph_9 Footprints Subgraph_2 PrimaryInput)
  (connect Subgraph_10 Footprints Subgraph_2 PrimaryInput)
  (connect Subgraph_13 Footprints Subgraph_2 PrimaryInput)
  (connect Subgraph_8 Unselected Subgraph_3 PrimaryInput)
  (connect AddTags_0 Out Subgraph_5 PrimaryInput)
  (connect Subgraph_1 Out Subgraph_5 SecondaryInput)
  (connect Subgraph_8 Selected Subgraph_6 PrimaryInput)
  (connect Subgraph_5 Areas Subgraph_7 PrimaryInput)
  (connect Subgraph_7 Areas Subgraph_8 PrimaryInput)
  (connect Subgraph_7 Subdivide%20Splines Subgraph_9 PrimaryInput)
  (connect Subgraph_6 Subdividing_splines Subgraph_10 PrimaryInput)
  (connect AddTags_0 Out Subgraph_13 PrimaryInput)
  (connect Subgraph_5 Subdividing%20Splines Subgraph_13 PrimaryInput)
  (output Out Subgraph_2)
  (output Out Subgraph_3))
