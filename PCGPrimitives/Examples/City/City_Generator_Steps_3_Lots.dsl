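; PCG graph for the lots step of the example city generator, turning areas into tagged building lots.
; It composes an exclusion from two actor-referenced data sets, then randomly selects a share set by buildableLots.
; Selected areas pass two size filters, get recursively split, then are size-filtered again.
; Both final filter results, the randomly unselected areas and the full exclusion result are tagged separately and output, with debug display on accepted lots.
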(graph City_Generator_Steps_3_Lots
  (params (buildableLots :type Double :default 0.9))
  (bind Subgraph_12 (Subgraph :Graph /PCGPrimitives/Primitives/Get/Get_ActorDataByRef.Get_ActorDataByRef))
  (bind Subgraph_15 (Subgraph :Graph /PCGPrimitives/Primitives/Get/Get_ActorDataByRef.Get_ActorDataByRef))
  (bind Subgraph_0 (Subgraph :Graph /PCGPrimitives/Primitives/Compose/Compose_Areas_Exclusion.Compose_Areas_Exclusion))
  (bind Subgraph_6 (Subgraph :Graph /PCGPrimitives/Primitives/Subdivide/Subdivide_Areas_WithRecursiveSplit.Subdivide_Areas_WithRecursiveSplit))
  (bind Subgraph_7 (Subgraph :Graph /PCGPrimitives/Primitives/Filter/Filter_Areas_BySize.Filter_Areas_BySize))
  (bind Subgraph_11 (Subgraph :Graph /PCGPrimitives/Primitives/Write/Write_Tag.Write_Tag))
  (bind Subgraph_13 (Subgraph :Graph /PCGPrimitives/Primitives/Write/Write_Tag.Write_Tag))
  (bind Subgraph_16 (Subgraph :Graph /PCGPrimitives/Primitives/Write/Write_Tag.Write_Tag))
  (bind Subgraph_17 (Subgraph :Graph /PCGPrimitives/Primitives/Write/Write_Tag.Write_Tag))
  (bind Subgraph_4 (Subgraph :Graph /PCGPrimitives/Primitives/Debug/Debug_PrimitiveData.Debug_PrimitiveData))
  ;[0.15 0.15 0.15 0.5] LOTS
  (bind Subgraph_3 (Subgraph :Graph /PCGPrimitives/Primitives/Filter/Filter_RandomChoice.Filter_RandomChoice :Ratio buildableLots))
  (bind Subgraph_1 (Subgraph :Graph /PCGPrimitives/Primitives/Filter/Filter_Areas_BySize.Filter_Areas_BySize))
  (bind Subgraph_2 (Subgraph :Graph /PCGPrimitives/Primitives/Filter/Filter_Areas_BySize.Filter_Areas_BySize))
  (connect Subgraph_15 Out Subgraph_0 PrimaryInput)
  (connect Subgraph_12 Out Subgraph_0 SecondaryInput)
  (connect Subgraph_3 Selected Subgraph_1 PrimaryInput)
  (connect Subgraph_1 Selected Subgraph_2 PrimaryInput)
  (connect Subgraph_0 Out Subgraph_3 PrimaryInput)
  (connect Subgraph_11 Out Subgraph_4 PrimaryInput)
  (connect Subgraph_2 Selected Subgraph_6 PrimaryInput)
  (connect Subgraph_6 Areas Subgraph_7 PrimaryInput)
  (connect Subgraph_7 Selected Subgraph_11 PrimaryInput)
  (connect Subgraph_7 Unselected Subgraph_13 PrimaryInput)
  (connect Subgraph_3 Unselected Subgraph_16 PrimaryInput)
  (connect Subgraph_0 Out Subgraph_17 PrimaryInput)
  (output Out Subgraph_11)
  (output Out Subgraph_13)
  (output Out Subgraph_16)
  (output Out Subgraph_17))
