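; PCG graph for the ground step (step 4) of the example city generator, built entirely from PCGPrimitives subgraphs.
; It fetches actor data by reference, filters it by tags and feeds the selection into two branches.
; One expands the areas, fills them uniformly and spawns assets.
; The other extracts segments, drops too-short splines as a fail-safe and assigns a shape grammar definition, whose output connects to nothing further here.
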(graph City_Generator_Steps_4_Ground
  (bind Subgraph_0 (Subgraph :Graph /PCGPrimitives/Primitives/Get/Get_ActorDataByRef.Get_ActorDataByRef))
  (bind Subgraph_1 (Subgraph :Graph /PCGPrimitives/Primitives/Filter/Filter_Tags.Filter_Tags))
  (bind Subgraph_3 (Subgraph :Graph /PCGPrimitives/Primitives/Transform/Transform_Areas_Expand.Transform_Areas_Expand))
  (bind Subgraph_9 (Subgraph :Graph /PCGPrimitives/Primitives/Fill/Fill_Areas_Uniformly.Fill_Areas_Uniformly))
  (bind Subgraph_11 (Subgraph :Graph /PCGPrimitives/Primitives/Spawn/Spawn_Assets.Spawn_Assets))
  (bind Subgraph_2 (Subgraph :Graph /PCGPrimitives/Primitives/Extract/Extract_Segments.Extract_Segments))
  (bind Subgraph_4 (Subgraph :Graph /PCGPrimitives/Primitives/Filter/Filter_Splines_byLength.Filter_Splines_byLength)) ; fail safe if too short
  (bind Subgraph_12 (Subgraph :Graph /PCGPrimitives/Primitives/Assign/Assign_ShapeGrammarDefinition.Assign_ShapeGrammarDefinition))
  (connect Subgraph_0 Out Subgraph_1 PrimaryInput)
  (connect Subgraph_1 Selected Subgraph_2 PrimaryInput)
  (connect Subgraph_1 Selected Subgraph_3 PrimaryInput)
  (connect Subgraph_2 Out Subgraph_4 PrimaryInput)
  (connect Subgraph_3 Out Subgraph_9 PrimaryInput)
  (connect Subgraph_9 Out Subgraph_11 PrimaryInput)
  (connect Subgraph_4 Selected Subgraph_12 PrimaryInput))
