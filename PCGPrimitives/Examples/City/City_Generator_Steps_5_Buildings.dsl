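; PCG graph for the buildings step of the City Generator example, turning tag-filtered actor data into buildings and outputting their footprints.
; Selected data gets an attribute from a texture, then the Debug parameter routes it to simple mesh extrusion instead of the full build.
; The full build assigns a shape grammar and, in parallel, offsets the data twice to create planar meshes and contracted areas.
; Those areas are uniformly filled, randomly thinned and spawned with assets.
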(graph City_Generator_Steps_5_Buildings
  (params (Debug :type Bool))
  (bind Subgraph_1 (Subgraph :Graph /PCGPrimitives/Primitives/Get/Get_ActorDataByRef.Get_ActorDataByRef))
  (bind Subgraph_0 (Subgraph :Graph /PCGPrimitives/Primitives/Filter/Filter_Tags.Filter_Tags))
  (bind Subgraph_2 (Subgraph :Graph /PCGPrimitives/Primitives/Write/Write_Attribute_fromTexture.Write_Attribute_fromTexture))
  (bind Branch_0 (Branch :bOutputToB-param Debug))
  (bind Subgraph_7 (Subgraph :Graph /PCGPrimitives/Primitives/Assign/Assign_ShapeGrammarDefinition.Assign_ShapeGrammarDefinition))
  (bind Subgraph_8 (Subgraph :Graph /PCGPrimitives/Primitives/Transform/Transform_Offset.Transform_Offset))
  (bind Subgraph_10 (Subgraph :Graph /PCGPrimitives/Primitives/Transform/Transform_Offset.Transform_Offset))
  (bind Subgraph_9 (Subgraph :Graph /PCGPrimitives/Primitives/Create/Create_Mesh_Planar.Create_Mesh_Planar))
  (bind Subgraph_14 (Subgraph :Graph /PCGPrimitives/Primitives/Transform/Transform_Areas_Contract.Transform_Areas_Contract))
  (bind Subgraph_11 (Subgraph :Graph /PCGPrimitives/Primitives/Fill/Fill_Areas_Uniformly.Fill_Areas_Uniformly))
  (bind Subgraph_12 (Subgraph :Graph /PCGPrimitives/Primitives/Filter/Filter_RandomChoice.Filter_RandomChoice))
  (bind Subgraph_13 (Subgraph :Graph /PCGPrimitives/Primitives/Spawn/Spawn_Assets.Spawn_Assets))
  (bind Subgraph_3 (Subgraph :Graph /PCGPrimitives/Primitives/Create/Create_Mesh_Extrude.Create_Mesh_Extrude))
  (connect Subgraph_2 Out Branch_0 In)
  (connect Subgraph_1 Out Subgraph_0 PrimaryInput)
  (connect Subgraph_0 Selected Subgraph_2 PrimaryInput)
  (connect Branch_0 Output%20B Subgraph_3 PrimaryInput)
  (connect Branch_0 Output%20A Subgraph_7 PrimaryInput)
  (connect Branch_0 Output%20A Subgraph_8 PrimaryInput)
  (connect Subgraph_10 Out Subgraph_9 PrimaryInput)
  (connect Subgraph_8 Out Subgraph_10 PrimaryInput)
  (connect Subgraph_14 Out Subgraph_11 PrimaryInput)
  (connect Subgraph_11 Out Subgraph_12 PrimaryInput)
  (connect Subgraph_12 Selected Subgraph_13 PrimaryInput)
  (connect Subgraph_10 Out Subgraph_14 PrimaryInput)
  (output Out Subgraph_7.Footprints))
