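; PCG graph for step 6 of the example city generator, which populates left-over lots with spawned assets, built entirely from PCGPrimitives subgraphs.
; Two actor-data-by-reference fetches feed tag filters that select the areas to work on.
; One branch randomly fills tagged areas plus the result of an area exclusion between two tag selections, then spawns assets.
; The plaza branch contracts tagged areas, fills them uniformly, keeps a random selection and spawns assets.
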(graph City_Generator_Steps_6_Left_over_lots
  (bind Subgraph_3 (Subgraph :Graph /PCGPrimitives/Primitives/Get/Get_ActorDataByRef.Get_ActorDataByRef))
  (bind Subgraph_5 (Subgraph :Graph /PCGPrimitives/Primitives/Filter/Filter_Tags.Filter_Tags))
  (bind Subgraph_6 (Subgraph :Graph /PCGPrimitives/Primitives/Get/Get_ActorDataByRef.Get_ActorDataByRef))
  (bind Subgraph_7 (Subgraph :Graph /PCGPrimitives/Primitives/Filter/Filter_Tags.Filter_Tags))
  (bind Subgraph_9 (Subgraph :Graph /PCGPrimitives/Primitives/Filter/Filter_Tags.Filter_Tags))
  (bind Subgraph_8 (Subgraph :Graph /PCGPrimitives/Primitives/Compose/Compose_Areas_Exclusion.Compose_Areas_Exclusion))
  (bind Subgraph_1 (Subgraph :Graph /PCGPrimitives/Primitives/Fill/Fill_Areas_Randomly.Fill_Areas_Randomly))
  (bind Subgraph_0 (Subgraph :Graph /PCGPrimitives/Primitives/Spawn/Spawn_Assets.Spawn_Assets))
  (bind Subgraph_4 (Subgraph :Graph /PCGPrimitives/Primitives/Filter/Filter_Tags.Filter_Tags))
  ;[0.15 0.15 0.15 0.5] Plaza
  (bind Subgraph_13 (Subgraph :Graph /PCGPrimitives/Primitives/Transform/Transform_Areas_Contract.Transform_Areas_Contract))
  (bind Subgraph_10 (Subgraph :Graph /PCGPrimitives/Primitives/Fill/Fill_Areas_Uniformly.Fill_Areas_Uniformly))
  (bind Subgraph_11 (Subgraph :Graph /PCGPrimitives/Primitives/Filter/Filter_RandomChoice.Filter_RandomChoice))
  (bind Subgraph_12 (Subgraph :Graph /PCGPrimitives/Primitives/Spawn/Spawn_Assets.Spawn_Assets))
  (connect Subgraph_1 Out Subgraph_0 PrimaryInput)
  (connect Subgraph_5 Selected Subgraph_1 PrimaryInput)
  (connect Subgraph_8 Out Subgraph_1 PrimaryInput)
  (connect Subgraph_3 Out Subgraph_4 PrimaryInput)
  (connect Subgraph_3 Out Subgraph_5 PrimaryInput)
  (connect Subgraph_6 Out Subgraph_7 PrimaryInput)
  (connect Subgraph_7 Selected Subgraph_8 PrimaryInput)
  (connect Subgraph_9 Selected Subgraph_8 SecondaryInput)
  (connect Subgraph_6 Out Subgraph_9 PrimaryInput)
  (connect Subgraph_13 Out Subgraph_10 PrimaryInput)
  (connect Subgraph_10 Out Subgraph_11 PrimaryInput)
  (connect Subgraph_11 Selected Subgraph_12 PrimaryInput)
  (connect Subgraph_4 Selected Subgraph_13 PrimaryInput))
