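; PCG example graph for step 7 of the city generator, which handles the city border contour.
; It chains three reusable primitive subgraphs: Get_ActorData feeds Transform_Areas_Expand, which feeds Assign_ShapeGrammarDefinition.
; The actor's data is thus expanded as areas and tagged with a shape grammar definition, with no further nodes consuming the result in this graph.
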(graph City_Generator_Steps_7_city_Contour
  (bind Subgraph_2 (Subgraph :Graph /PCGPrimitives/Primitives/Get/Get_ActorData.Get_ActorData))
  ;[0.15 0.15 0.15 0.5] CITY BORDER
  (bind Subgraph_0 (Subgraph :Graph /PCGPrimitives/Primitives/Transform/Transform_Areas_Expand.Transform_Areas_Expand))
  (bind Subgraph_1 (Subgraph :Graph /PCGPrimitives/Primitives/Assign/Assign_ShapeGrammarDefinition.Assign_ShapeGrammarDefinition))
  (connect Subgraph_2 Out Subgraph_0 PrimaryInput)
  (connect Subgraph_0 Out Subgraph_1 PrimaryInput))
